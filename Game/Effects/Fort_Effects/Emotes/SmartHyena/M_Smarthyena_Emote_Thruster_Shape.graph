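# Additive unlit Niagara sprite material, named as the thruster shape effect for the SmartHyena emote.
# Particle color RGB feeds the MF_FX_EmissiveAdjustments function to shape the glow color.
# Particle alpha is multiplied by the green channel of the T_Smarthyena_Shape texture, masking the sprite to that shape.
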Material /Game/Effects/Fort_Effects/Emotes/SmartHyena/M_Smarthyena_Emote_Thruster_Shape
graph {
 properties {bUsedWithNiagaraSprites=true, blendMode=BLEND_Additive, pixelDepthOffsetMode=PDOM_Legacy, shadingModel=MSM_Unlit, usedShadingModels=Unlit}
 nodes [
  n1 = ParticleColor()
  MaterialFunctionCall(Emissive=@n1.RGB) {materialFunction=&MF_FX_EmissiveAdjustments}
  Multiply(A=@n1.A, B=TextureSample() {samplerType=SAMPLERTYPE_LinearColor, texture=&/T_Smarthyena_Shape}.G)
 ]
}
