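# Translucent unlit Niagara ribbon and sprite material that draws wind streaks for the SmartHyena emote.
# A panning, per-particle randomized swirl flowmap distorts the UVs of streak and smoke textures, blended by dynamic parameters.
# Particle color, an optional time-of-day tint and a fire-colored Temperature term feed emissive.
# Opacity combines a smoothstep dissolve, optional depth and near-camera fades, and a sine-based soft edge mask.
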Material /Game/Effects/Fort_Effects/Emotes/SmartHyena/M_Wind_Streak
graph {
 properties {
  bAllowNegativeEmissiveColor=true
  bUsedWithNanite=true
  bUsedWithNiagaraRibbons=true
  bUsedWithNiagaraSprites=true
  blendMode=BLEND_Translucent
  pixelDepthOffsetMode=PDOM_Legacy
  shadingModel=MSM_Unlit
  usedShadingModels=Unlit
 }
 nodes [
  n1 = ParticleColor()
  n2 = ScalarParameter() {defaultValue=0.002, parameterName=Swirl UVW Scale}
  n3 = Multiply(A=@n2) {constB=1000}
  n4 = DynamicParameter() {parameterIndex=1, defaultValue=<rgba 0 0 1 0>, paramNames [Swirl Distortion, Emissive, Diffuse Power, Swirl UV Pan]}
  n5 = Add(A=AppendVector(A=ParticleRandom(), B=@n4.Swirl UV Pan), B=TextureCoordinate())
  n6 = ComponentMask(None=TextureSample(UVs=Multiply(A=@n3, B=@n5)) {samplerType=SAMPLERTYPE_LinearColor, texture=/Game/VisualThreatManager/StormVisuals/LargeClouds/T_Swirl_Flowmap}.RGB) {g=true, r=true}
  n7 = DynamicParameter() {parameterIndex=1, defaultValue=<rgba 0 0 1 0>, paramNames [Swirl Distortion, Emissive, Diffuse Power, Swirl UV Pan]}
  n8 = Add(A=Multiply(A=Subtract(A=@n6) {constB=0.5}, B=@n7.Swirl Distortion), B=TextureCoordinate())
  n9 = TextureSampleParameter2D(UVs=@n8) {
   parameterName=Sterak Texture
   samplerType=SAMPLERTYPE_LinearColor
   texture=&/T_Smoke_Flipbook_Handpainted
  }
  n10 = TextureSampleParameter2D(UVs=@n8) {parameterName=Smoke Texture, samplerType=SAMPLERTYPE_LinearColor, texture=/Game/Effects/Explosion_Update/Textures/T_Smoke_Streak}
  n11 = DynamicParameter() {defaultValue=<rgba 1 1 10 0>, paramNames [DepthFade, BP_Intensity, BP_HueShift, Shift_to_Smoke]}
  n12 = Multiply(A=@n10.R, B=@n11.Shift_to_Smoke)
  n13 = LinearInterpolate(A=@n9.A, B=@n10.RGB, Alpha=Saturate(None=Add(A=@n12, B=@n11.Shift_to_Smoke)))
  n14 = DynamicParameter() {parameterIndex=1, defaultValue=<rgba 0 0 1 0>, paramNames [Swirl Distortion, Emissive, Diffuse Power, Swirl UV Pan]}
  n15 = Multiply(A=@n1.RGB, B=Power(Base=@n13, Exp=@n14.Diffuse Power))
  n16 = Multiply(
   A=MaterialFunctionCall(
    Color_Morning=VectorParameter() {parameterName=Color Morning, defaultValue=<rgba 0.2795 0.1398 0.1407 1>}.RGB
    Color_Day=VectorParameter() {parameterName=Color Day, defaultValue=<rgba 0.6094 0.545 0.5552 1>}.RGB
    Color_Evening=VectorParameter() {parameterName=Color Evening, defaultValue=<rgba 0.2726 0.1761 0.1282 1>}.RGB
    Color_Night=VectorParameter() {parameterName=Color Night, defaultValue=<rgba 0.3438 0.3438 0.3438 1>}.RGB
   ) {materialFunction=&Color_Time_of_Day}.Result
   B=@n15
  )
  MaterialFunctionCall(Emissive=Add(A=StaticSwitchParameter(True=@n16, False=@n15) {parameterName=UseTimeOfDayColor}, B=NamedRerouteUsage())) {materialFunction=&MF_FX_EmissiveAdjustments}
  n17 = DynamicParameter() {parameterIndex=2, defaultValue=<rgba 0.75 0 1 0>, paramNames [Smoothness Dissolve, Dissolve, Param3, Param4]}
  n18 = OneMinus(None=@n17.Dissolve)
  n19 = Add(A=@n18, B=@n17.Smoothness Dissolve)
  n20 = Multiply(
   A=Multiply(
    A=MaterialFunctionCall(Alpha=@n9.G, Min=@n18, Max=@n19) {materialFunction=&SmoothStep}.Result
    B=@n1.A
   )
   B=@n13
  )
  n21 = LinearInterpolate(
   A=ScalarParameter() {defaultValue=5, group=Opacity|DepthFade, parameterName=DepthFade_Distance, sortPriority=2}
   B=@n11.DepthFade
   Alpha=ScalarParameter() {defaultValue=1, group=Opacity|DepthFade, parameterName=Use Dynamic Parameter to Control Depth Fade, sortPriority=1}
  )
  n22 = LinearInterpolate(
   A=@n20
   B=ShadingPathSwitch(Default=DepthFade(Opacity=@n20, FadeDistance=@n21), Mobile=@n20)
   Alpha=ScalarParameter() {group=Opacity|DepthFade, parameterName=Use Depth Fade}
  )
  n23 = ScalarParameter() {defaultValue=512, group=Opacity|Near_Camera_Fade, parameterName=NearCamFade_Distance, sortPriority=1}
  n24 = Clamp(None=Power(
    Base=MaterialFunctionCall(Distance=@n23) {materialFunction=&Near_Camera_Fade}.Result
    Exp=ScalarParameter() {defaultValue=1, group=Opacity|Near_Camera_Fade, parameterName=NearCamFade_Power, sortPriority=2}
   )) {minDefault=0.02}
  n25 = VectorParameter() {parameterName=Edge Mask UV Multiply, channelNames {a="", b="", g=V, r=U}, defaultValue=<rgba 4 1 0 0>}
  n26 = Multiply(A=Multiply(A=TextureCoordinate(), B=ComponentMask(None=@n25) {g=true, r=true})) {constB=0.5}
  n27 = MaterialFunctionCall(Float2=Abs(None=Sine(None=@n26))) {materialFunction=&BreakOutFloat2Components}
  Multiply(
   A=LinearInterpolate(A=@n22, B=Multiply(A=@n22, B=@n24), Alpha=ScalarParameter() {group=Opacity|Near_Camera_Fade, parameterName=Add Near Camera Fade})
   B=LinearInterpolate(B=Multiply(A=@n27.R, B=@n27.R), Alpha=ScalarParameter() {defaultValue=1, group=Opacity, parameterName=SoftenEdge}) {constA=1}
  )
  n28 = DynamicParameter() {parameterIndex=1, defaultValue=<rgba 0 0 1 0>, paramNames [Swirl Distortion, Emissive, Diffuse Power, Swirl UV Pan]}
  n29 = Add(A=Multiply(A=@n28.Emissive) {constB=0.1}, B=Multiply(A=@n28.Emissive, B=@n10.R))
  NamedRerouteDeclaration(None=Multiply(A=VectorParameter() {parameterName=Fire Color, defaultValue=<rgba 16 1.336 0 0>}.RGB, B=@n29)) {name=Temperature, nodeColor=<rgba 1 0.8438 0 1>}
  Saturate(None=@n29)
  n30 = Add(A=WorldPosition().XYZ) {constB=-50000}
  TextureSample(UVs=Multiply(A=Add(A=@n30, B=@n4.Swirl UV Pan), B=@n2) {constB=0.0004}) {samplerType=SAMPLERTYPE_LinearColor, texture=/Game/Effects/FogVolume/VT_CurlNoise_LOW}
 ]
}
